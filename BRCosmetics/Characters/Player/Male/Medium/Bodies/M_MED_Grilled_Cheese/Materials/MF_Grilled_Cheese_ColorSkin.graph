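# Material function that recolors the skin of the Grilled Cheese male body cosmetic within incoming material attributes.
# A partly desaturated SkinColor is modulated by the base color's HSV value, keeping texture shading, with a blend back to flat color.
# A Fresnel rim of SkinColor forms the skin emissive.
# Both blend over the original BaseColor and Emissive through a mask from either an FX texture channel or the Skin input, chosen by static switch.
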MaterialFunction /BRCosmetics/Characters/Player/Male/Medium/Bodies/M_MED_Grilled_Cheese/Materials/MF_Grilled_Cheese_ColorSkin
graph {
 properties {description=SkinColorSaturation}
 nodes [
  n1 = FunctionInput() {inputType=FunctionInput_MaterialAttributes}
  n2 = GetMaterialAttributes(None=@n1) {attributeGetTypes [#1, #2]}
  n3 = VectorParameter() {group=GrilledCheeseSkin, parameterName=SkinColor, defaultValue=<rgba 0.8 0.05 0 0>}
  n4 = Desaturation(None=@n3.RGB, Fraction=ScalarParameter() {defaultValue=0.15, group=GrilledCheeseSkin, parameterName=SkinColorDesaturation})
  n5 = MaterialFunctionCall(Float3=MaterialFunctionCall!(In=@n2.BaseColor) {materialFunction=/Engine/Functions/Engine_MaterialFunctions01/Chromakeying/RGBtoHSV}.Result) {materialFunction=&BreakOutFloat3Components}
  n6 = ScalarParameter() {
   defaultValue=1.3
   desc="Mults the V strength of the skin diffuse color. \r\nUsed by the tint color."
   group=GrilledCheeseSkin
   parameterName=SkinVMult
  }
  n7 = ScalarParameter() {
   defaultValue=0.9
   desc="0=No skin texture used. 1=Value of skin texture multed against skinColor"
   group=GrilledCheeseSkin
   parameterName=SkinVStength
  }
  n8 = LinearInterpolate(A=@n4, B=Multiply(A=Multiply(A=@n5.B, B=@n6), B=@n4), Alpha=@n7) {constAlpha=0.3}
  n9 = TextureSampleParameter2D() {
   parameterName=Skin FX
   samplerType=SAMPLERTYPE_Masks
   texture=/BRCosmetics/Characters/Player/Male/Medium/Bodies/M_MED_Grilled_Cheese/Textures/T_M_MED_Grilled_Cheese_Body_FX
  }
  n10 = Multiply(
   A=StaticSwitchParameter(
    True=ChannelMaskParameter(None=@n9.RGBA) {group=GrilledCheeseSkin, parameterName=ColorSkin_SkinFXMask}.R
    False=FunctionInput() {inputName=Skin, inputType=FunctionInput_Scalar}
   ) {group=GrilledCheeseSkin, parameterName=ColorSkin_useSkinFX}
   B=ScalarParameter() {defaultValue=1, group=GrilledCheeseSkin, parameterName=ColorSkinStrength}
  ) {constB=0.5}
  n11 = Multiply(
   A=Multiply(A=Fresnel() {baseReflectFraction=0, exponent=2}, B=@n3.RGB)
   B=ScalarParameter() {group=GrilledCheeseSkin, parameterName=SkinEmissiveStrength}
  )
  FunctionOutput(None=SetMaterialAttributes!(
    MaterialAttributes=@n1
    BaseColor=LinearInterpolate(A=@n2.BaseColor, B=@n8, Alpha=@n10)
    EmissiveColor=LinearInterpolate(A=@n2.EmissiveColor, B=@n11, Alpha=@n10)
   ) {attributeSetTypes [#1, #2]})
 ]
}
