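# Material function that layers an animated, revealable emissive effect onto the Grilled Cheese character's input material attributes.
# A noise texture is triplanar-projected from pre-skinned position and normal, so it follows the mesh while scrolling and distorting.
# It drives brightness and darkens base color, with fresnel blending two emissive colors.
# Gradient, reveal and primary mask control the blend, plus a secondary mask for hair; the curve-atlas ramp is unconnected.
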MaterialFunction /BRCosmetics/Characters/Player/Male/Medium/Bodies/M_MED_Grilled_Cheese/Materials/MF_Grilled_Cheese_Emissive
graph {nodes [
  n1 = Multiply(
   A=FunctionInput() {bUsePreviewValueAsDefault=true, inputName=UVOffset, inputType=FunctionInput_Scalar}
   B=AppendVector(A=ScalarParameter() {group=GrilledCheese, parameterName=UVOffsetU}, B=ScalarParameter() {group=GrilledCheese, parameterName=UVOffsetV})
  )
  n2 = ScalarParameter() {defaultValue=1, group=GrilledCheese, parameterName=Emissive Tiling}
  n3 = PreSkinnedPosition()
  n4 = Divide(
   A=FunctionInput(Preview=VertexInterpolator(VS=@n3).PS) {bUsePreviewValueAsDefault=true, inputName=Local Position}
   B=ScalarParameter() {defaultValue=40, group=GrilledCheese, parameterName=Triplanner_ConstantScaleDivisor}
  )
  n5 = Frac(None=Multiply(A=Time(), B=VectorParameter() {group=GrilledCheese, parameterName=TriPlanner Speed, defaultValue=<rgba 0 0.2 0.2 1>}.RGB))
  n6 = Add(A=Multiply(A=@n2, B=@n4), B=@n5)
  n7 = Panner() {speedX=0.01, speedY=0.01}
  n8 = Multiply(
   A=ComponentMask(None=TextureSample(UVs=@n7) {samplerType=SAMPLERTYPE_Masks, texture=/Game/Global/Textures/Default/Noise/T_Pattern_Noise_M}.RGB) {b=true, g=true}
   B=ScalarParameter() {defaultValue=0.1, group=GrilledCheese, parameterName=EmissiveDistortion}
  ) {constB=0.1}
  n9 = Add(A=Add(A=@n1, B=ComponentMask(None=@n6) {b=true, g=true}), B=@n8)
  n10 = TextureObjectParameter() {
   group=GrilledCheese
   parameterName=Panning Texture
   samplerType=SAMPLERTYPE_Masks
   texture=&T_Pattern_Noise_M
  }
  n11 = Add(A=Multiply(A=@n2, B=@n4), B=@n5)
  n12 = Add(A=Add(A=@n1, B=ComponentMask(None=@n11) {b=true, r=true}), B=@n8)
  n13 = FunctionInput(Preview=VertexInterpolator(VS=PreSkinnedNormal()).PS) {bUsePreviewValueAsDefault=true, inputName=Local Normal}
  n14 = ComponentMask(None=@n13) {g=true}
  n15 = ScalarParameter() {defaultValue=1, group=GrilledCheese, parameterName=TriPlanner_AxisBlendStrength}
  n16 = ComponentMask(None=MaterialFunctionCall(In=Abs(None=@n14), Contrast=@n15) {materialFunction=&CheapContrast}.Result) {g=true}
  n17 = Multiply(A=@n2, B=@n4)
  n18 = Add(A=@n1, B=ComponentMask(None=Add(A=@n17, B=@n5)) {g=true, r=true})
  n19 = MaterialFunctionCall(In=Abs(None=ComponentMask(None=@n13) {b=true}), Contrast=@n15) {materialFunction=&CheapContrast}
  n20 = LinearInterpolate(
   A=LinearInterpolate(
    A=TextureSample(UVs=@n9, Tex=@n10) {texture=/Game/Effects/Fort_Effects/Textures/Grad/T_FX_LinearSweep_Gradient}.RGB
    B=TextureSample(UVs=@n12, Tex=@n10) {texture=&T_FX_LinearSweep_Gradient}.RGB
    Alpha=Saturate(None=@n16)
   )
   B=TextureSample(UVs=Add(A=@n18, B=@n8), Tex=@n10) {texture=&T_FX_LinearSweep_Gradient}.RGB
   Alpha=Saturate(None=ComponentMask(None=@n19.Result) {b=true})
  )
  n21 = Add(A=@n20, B=Multiply(A=Time()) {constB=0.1})
  CurveAtlasRowParameter(CurveTime=ComponentMask(None=Frac(None=@n21)) {r=true}) {atlas=&AthenaCurveAtlas01, curve=/Game/Athena/LinearColorCurves/Curve_Neon_Trail_Ramp, defaultValue=65}
  n22 = FunctionInput() {inputName=Material, inputType=FunctionInput_MaterialAttributes}
  n23 = GetMaterialAttributes(None=@n22) {attributeGetTypes [#1]}
  n24 = VectorParameter() {group=GrilledCheese, parameterName=Emissive Color A, defaultValue=<rgba 1 0.9995 0.1354 1>}
  n25 = VectorParameter() {group=GrilledCheese, parameterName=Emissive Color B, defaultValue=<rgba 0 0.7217 1 1>}
  n26 = Fresnel(ExponentIn=ScalarParameter() {defaultValue=1, group=GrilledCheese, parameterName=Fresnel_ExponentIn}) {baseReflectFraction=0, exponent=1}
  n27 = ScalarParameter() {defaultValue=0.01, group=GrilledCheese, parameterName=Brightness_Min}
  n28 = ScalarParameter() {defaultValue=8, group=GrilledCheese, parameterName=Brightness_Max}
  n29 = ComponentMask(None=@n20) {r=true}
  n30 = Multiply(
   A=Multiply(A=LinearInterpolate(A=@n24.RGB, B=@n25.RGB, Alpha=@n26), B=LinearInterpolate(A=@n27, B=@n28, Alpha=@n29))
   B=ScalarParameter() {defaultValue=1, group=GrilledCheese, parameterName=Brightness_Multiplier}
  ) {constB=100}
  n31 = Multiply(A=@n23.BaseColor, B=@n29)
  n32 = Add(
   A=FunctionInput() {inputName=Gradient Mask, inputType=FunctionInput_Scalar}
   B=ScalarParameter() {group=GrilledCheese, parameterName=EmissiveReveal}
  )
  n33 = Multiply(A=Saturate(None=Subtract(A=@n32)), B=FunctionInput() {inputName=Primary Mask, inputType=FunctionInput_Scalar})
  n34 = LinearInterpolate(A=Multiply(A=@n33), B=Multiply(A=@n33, B=FunctionInput() {inputName=Secondary Mask, inputType=FunctionInput_Scalar})) {constAlpha=0.95}
  FunctionOutput(None=BlendMaterialAttributes(
    A=@n22
    B=SetMaterialAttributes!(MaterialAttributes=@n23.MaterialAttributes, EmissiveColor=@n30, BaseColor=@n31) {attributeSetTypes [#2, #1]}
    Alpha=StaticSwitchParameter(True=@n34, False=@n33) {group=GrilledCheese, parameterName=isHair}
   ))
 ]}
