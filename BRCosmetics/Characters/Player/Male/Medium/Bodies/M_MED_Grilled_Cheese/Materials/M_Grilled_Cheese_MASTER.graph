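# Master material for the Grilled Cheese medium male player skin, built on material attributes between the shared character PreFX and PostFX functions.
# Channel masks from the Skin FX texture drive a custom emissive function, whose result passes through a weapon-skin emissive Fresnel function.
# A color-skin function combines that result with the PreFX skin.
# An adjustable Fresnel term then lerps Base Color toward a tint before PostFX.
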Material /BRCosmetics/Characters/Player/Male/Medium/Bodies/M_MED_Grilled_Cheese/Materials/M_Grilled_Cheese_MASTER
graph {
 properties {
  bUseMaterialAttributes=true
  bUsedWithMeshDeformer=true
  bUsedWithMorphTargets=true
  bUsedWithSkeletalMesh=true
  pixelDepthOffsetMode=PDOM_Legacy
  usedShadingModels=Default Lit
 }
 nodes [
  n1 = MaterialFunctionCall() {materialFunction=&BaseFN_CharacterFunction_PreFX}
  n2 = TextureSampleParameter2D() {
   parameterName=Skin FX
   samplerType=SAMPLERTYPE_Masks
   texture=/BRCosmetics/Characters/Player/Male/Medium/Bodies/M_MED_Grilled_Cheese/Textures/T_M_MED_Grilled_Cheese_Body_FX
  }
  n3 = ChannelMaskParameter(None=@n2.RGBA) {group=GrilledCheese, parameterName=Secondary Mask}
  n4 = ChannelMaskParameter(None=@n2.RGBA) {group=GrilledCheese, maskChannel=Alpha, parameterName=Gradient Emissive Mask, defaultValue=<rgba 0 0 0 1>}
  n5 = MaterialFunctionCall!(
   Primary Mask=ChannelMaskParameter(None=@n2.RGBA) {group=GrilledCheese, parameterName=Primary Mask}.R
   Secondary Mask=@n3.R
   Gradient Mask=@n4.R
   Local Normal
   Local Position=@n1.Pre skinned Local Position
   UVOffset=@n3.R
   Material=@n1.Pre skinned Local Position
  ) {materialFunction=&/MF_Grilled_Cheese_Emissive}
  n6 = GetMaterialAttributes(None=MaterialFunctionCall!(
    Skin=@n1.Skin
    In=MaterialFunctionCall!(In=@n5.Result) {materialFunction=/Game/Weapons/WeaponSkins/MaterialFunction/MF_EmissiveFresnel}.Result
   ) {materialFunction=&/MF_Grilled_Cheese_ColorSkin}.Result) {attributeGetTypes [#1]}
  n7 = VectorParameter() {group=GrilledCheeseFresnel, parameterName=FresnelTint, defaultValue=<rgba 1 0.1 0.2 1>}
  n8 = Multiply(
   A=Fresnel(ExponentIn=ScalarParameter() {defaultValue=1.5, group=GrilledCheeseFresnel, parameterName=FresnelTintExponent}) {exponent=1.5}
   B=ScalarParameter() {defaultValue=1, group=GrilledCheeseFresnel, parameterName=FresnelTintStrength}
  )
  MaterialFunctionCall(
   AO=@n1.AO
   Distance Blend=@n1.AO
   Fresnel=@n1.AO
   Gmap=@n1.AO
   Pre Skinned Local Position=@n1.AO
   Material=SetMaterialAttributes!(MaterialAttributes=@n6.MaterialAttributes, Base Color=LinearInterpolate(A=@n6.Base Color, B=@n7.RGB, Alpha=@n8)) {attributeSetTypes [#1]}
  ) {materialFunction=&BaseFN_CharacterFunction_PostFX}
 ]
}
