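# Physics asset for the male medium Grilled Cheese player body cosmetic, covering its dangling front and back straps.
# Each strap is a three-box chain with angular limits, hung from a small capsule base constrained to the pelvis.
# Capsules on the pelvis and both thighs supply body volumes around the straps.
# Constraints from the pelvis to strap bases and thighs leave swing and twist free.
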PhysicsAsset /BRCosmetics/Characters/Player/Male/Medium/Bodies/M_MED_Grilled_Cheese/Meshes/M_MED_Grilled_Cheese_Physics
bodies {
 dyn_strap_bk_a [{shapeType=Box, extentX=4, extentY=1.5, extentZ=6, center=<0 0 -3.3195>}]
 dyn_strap_bk_a_base [{shapeType=Capsule, length=1, radius=0.5}]
 dyn_strap_bk_b [{shapeType=Box, extentX=4, extentY=1.5, extentZ=6, center=<0 0 -3.2619>}]
 dyn_strap_bk_c [{shapeType=Box, extentX=4, extentY=2, extentZ=6, center=<0 0 -2.7046>}]
 dyn_strap_fr_a [{shapeType=Box, extentX=4, extentY=1.5, extentZ=10.5, center=<0 0 -5.1616>}]
 dyn_strap_fr_a_base [{shapeType=Capsule, length=1, radius=0.5}]
 dyn_strap_fr_b [{shapeType=Box, extentX=5, extentY=1.5, extentZ=6.5, center=<0 0 -3.5634>}]
 dyn_strap_fr_c [{shapeType=Box, extentX=6.5, extentY=2, extentZ=9.5, center=<0 0 -4.8953>}]
 pelvis [{shapeType=Capsule, length=9, radius=10.5, center=<0 -1.391 0>}]
 thigh_l [{shapeType=Capsule, length=26.5, radius=8.5, center=<-20.9357 0 -0.9961>, rotation=<rot 87.7 180 180>}]
 thigh_r [{shapeType=Capsule, length=26.25, radius=8.75, center=<20.9357 0 0.9961>, rotation=<rot 87.6995 180 0>}]
}
constraints [
 {
  bone1Name=thigh_l
  bone2Name=pelvis
  swing1LimitDegrees=45
  swing1Motion=Free
  swing2LimitDegrees=45
  swing2Motion=Free
  twistLimitDegrees=45
  twistMotion=Free
 }
 {
  bone1Name=dyn_strap_bk_a_base
  bone2Name=pelvis
  swing1LimitDegrees=45
  swing1Motion=Free
  swing2LimitDegrees=45
  swing2Motion=Free
  twistLimitDegrees=45
  twistMotion=Free
 }
 {
  bone1Name=dyn_strap_fr_a_base
  bone2Name=pelvis
  swing1LimitDegrees=45
  swing1Motion=Free
  swing2LimitDegrees=45
  swing2Motion=Free
  twistLimitDegrees=45
  twistMotion=Free
 }
 {bone1Name=dyn_strap_bk_a, bone2Name=dyn_strap_bk_a_base, swing1LimitDegrees=30, swing2LimitDegrees=30, twistLimitDegrees=45}
 {bone1Name=dyn_strap_bk_b, bone2Name=dyn_strap_bk_a, swing1LimitDegrees=30, swing2LimitDegrees=30, twistLimitDegrees=45}
 {bone1Name=dyn_strap_bk_c, bone2Name=dyn_strap_bk_b, swing1LimitDegrees=30, swing2LimitDegrees=30, twistLimitDegrees=45}
 {bone1Name=dyn_strap_fr_a, bone2Name=dyn_strap_fr_a_base, swing1LimitDegrees=30, swing2LimitDegrees=30, twistLimitDegrees=45}
 {bone1Name=dyn_strap_fr_b, bone2Name=dyn_strap_fr_a, swing1LimitDegrees=30, swing2LimitDegrees=30, twistLimitDegrees=45}
 {bone1Name=dyn_strap_fr_c, bone2Name=dyn_strap_fr_b, swing1LimitDegrees=30, swing2LimitDegrees=30, twistLimitDegrees=45}
 {
  bone1Name=thigh_r
  bone2Name=pelvis
  swing1LimitDegrees=45
  swing1Motion=Free
  swing2LimitDegrees=45
  swing2Motion=Free
  twistLimitDegrees=45
  twistMotion=Free
 }
]
