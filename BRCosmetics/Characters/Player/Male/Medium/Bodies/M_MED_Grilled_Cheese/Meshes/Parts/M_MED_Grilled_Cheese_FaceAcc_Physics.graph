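# Physics asset for the Grilled Cheese male medium character's face accessory mesh, providing physics bodies for its dynamic hair bones.
# Nine identical small capsules on dyn_hair_a through dyn_hair_i each constrain directly to a head capsule, forming independent single-bone links rather than chains.
# Tight, equal swing and twist limits keep each hair piece to slight movement relative to the head.
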PhysicsAsset /BRCosmetics/Characters/Player/Male/Medium/Bodies/M_MED_Grilled_Cheese/Meshes/Parts/M_MED_Grilled_Cheese_FaceAcc_Physics
bodies {
 dyn_hair_a [{shapeType=Capsule, length=2, radius=1, center=<1.5 0 0>, rotation=<rot 90 0 0>}]
 dyn_hair_b [{shapeType=Capsule, length=2, radius=1, center=<1.5 0 0>, rotation=<rot 90 0 0>}]
 dyn_hair_c [{shapeType=Capsule, length=2, radius=1, center=<1.5 0 0>, rotation=<rot 90 0 0>}]
 dyn_hair_d [{shapeType=Capsule, length=2, radius=1, center=<1.5 0 0>, rotation=<rot 90 0 0>}]
 dyn_hair_e [{shapeType=Capsule, length=2, radius=1, center=<1.5 0 0>, rotation=<rot 90 0 0>}]
 dyn_hair_f [{shapeType=Capsule, length=2, radius=1, center=<1.5 0 0>, rotation=<rot 90 0 0>}]
 dyn_hair_g [{shapeType=Capsule, length=2, radius=1, center=<1.5 0 0>, rotation=<rot 90 0 0>}]
 dyn_hair_h [{shapeType=Capsule, length=2, radius=1, center=<1.5 0 0>, rotation=<rot 90 0 0>}]
 dyn_hair_i [{shapeType=Capsule, length=2, radius=1, center=<1.5 0 0>, rotation=<rot 90 0 0>}]
 head [{shapeType=Capsule, length=1, radius=1}]
}
constraints [
 {bone1Name=dyn_hair_a, bone2Name=head, swing1LimitDegrees=5, swing2LimitDegrees=5, twistLimitDegrees=5}
 {bone1Name=dyn_hair_b, bone2Name=head, swing1LimitDegrees=5, swing2LimitDegrees=5, twistLimitDegrees=5}
 {bone1Name=dyn_hair_c, bone2Name=head, swing1LimitDegrees=5, swing2LimitDegrees=5, twistLimitDegrees=5}
 {bone1Name=dyn_hair_d, bone2Name=head, swing1LimitDegrees=5, swing2LimitDegrees=5, twistLimitDegrees=5}
 {bone1Name=dyn_hair_e, bone2Name=head, swing1LimitDegrees=5, swing2LimitDegrees=5, twistLimitDegrees=5}
 {bone1Name=dyn_hair_f, bone2Name=head, swing1LimitDegrees=5, swing2LimitDegrees=5, twistLimitDegrees=5}
 {bone1Name=dyn_hair_g, bone2Name=head, swing1LimitDegrees=5, swing2LimitDegrees=5, twistLimitDegrees=5}
 {bone1Name=dyn_hair_h, bone2Name=head, swing1LimitDegrees=5, swing2LimitDegrees=5, twistLimitDegrees=5}
 {bone1Name=dyn_hair_i, bone2Name=head, swing1LimitDegrees=5, swing2LimitDegrees=5, twistLimitDegrees=5}
]
